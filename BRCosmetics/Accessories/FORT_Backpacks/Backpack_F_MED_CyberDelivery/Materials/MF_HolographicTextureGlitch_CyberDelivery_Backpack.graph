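# Material function for the CyberDelivery backpack that distorts input UVs with a horizontal holographic glitch.
# One GlitchMap sample uses rotated, stretched, panning UVs, while another scales V by GlitchScale and jumps in time steps via floored Time.
# Their smoothstepped red and biased green channels multiply into a mask, scaled by StretchedGlitchIntensity with a negative bias, then added to U.
# A trailing Panner node is unconnected.
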MaterialFunction /BRCosmetics/Accessories/FORT_Backpacks/Backpack_F_MED_CyberDelivery/Materials/MF_HolographicTextureGlitch_CyberDelivery_Backpack
graph {nodes [
  n1 = FunctionInput(Preview=TextureCoordinate()) {bUsePreviewValueAsDefault=true, inputName=UV, inputType=FunctionInput_Vector2}
  n2 = Multiply(
   A=Rotator(Coordinate=@n1, Time=ScalarParameter() {group=HolographicTexture_Glitch, parameterName=GlitchUVRotation})
   B=Constant2Vector() {g=10}
  )
  n3 = AppendVector(A=Constant(), B=ScalarParameter() {defaultValue=2, group=HolographicTexture_Glitch, parameterName=StretchedScrollingSpeed})
  n4 = TextureSample(UVs=Multiply(A=Constant2Vector() {g=0.6, r=1}, B=Panner(Coordinate=@n2, Speed=@n3) {speedX=0.5, speedY=5})) {
   samplerSource=SSM_Wrap_WorldGroupSettings
   samplerType=SAMPLERTYPE_LinearColor
   texture=/Game/Environments/Asteria/Sets/NeonCity/Techart/Holograms/GlitchMap_M
  }
  n5 = Constant() {r=1}
  n6 = FunctionInput(Preview=Constant() {r=2}) {bUsePreviewValueAsDefault=true, inputName=GlitchScale, inputType=FunctionInput_Scalar, sortPriority=1, previewValue=<2 0 0 1>}
  n7 = Constant() {r=8}
  n8 = Multiply(A=Time(), B=@n7) {constB=3}
  n9 = Add(A=Multiply(A=@n1, B=AppendVector(A=@n5, B=@n6)) {constB=5}, B=Divide(A=Floor(None=@n8), B=@n7) {constB=3})
  n10 = TextureSample(UVs=@n9) {
   samplerSource=SSM_Wrap_WorldGroupSettings
   samplerType=SAMPLERTYPE_LinearColor
   texture=&GlitchMap_M
  }
  n11 = Multiply(A=SmoothStep(Value=@n4.R) {constMax=0.5}, B=Saturate(None=Add(A=@n10.G) {constB=0.3}))
  n12 = ScalarParameter() {defaultValue=0.05, group=HolographicTexture_Glitch, parameterName=StretchedGlitchIntensity}
  n13 = ScalarParameter() {defaultValue=0.05, group=HolographicTexture_Glitch, parameterName=StretchedGlitchIntensity}
  n14 = Add(
   A=Multiply(A=Constant2Vector() {r=1}, B=Multiply(A=@n11, B=@n12) {constB=0.2})
   B=AppendVector(A=Multiply(B=@n13) {constA=-0.75}, B=Constant())
  )
  FunctionOutput(None=Add(A=@n1, B=Multiply(A=@n14) {constB=1.7})) {outputName=Glitched UV}
  Panner(Coordinate=@n9) {speedY=0.5}
 ]}
