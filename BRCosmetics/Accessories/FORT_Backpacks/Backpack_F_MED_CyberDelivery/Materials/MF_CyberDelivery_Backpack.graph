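# Material function for the CyberDelivery backpack that adds an animated holographic two-digit counter and cat logo to emissive, apparently reacting to kills.
# Blueprint-set digits select flipbook cells in a mask atlas, whose SDF thresholds yield solid shapes plus halftone and pulsing fringes.
# A Blueprint timestamp drives glitch, glow and reposition timing.
# The square-masked, scanlined result joins the original emissive, optionally cut out beneath the graphics.
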MaterialFunction /BRCosmetics/Accessories/FORT_Backpacks/Backpack_F_MED_CyberDelivery/Materials/MF_CyberDelivery_Backpack
graph {nodes [
  n1 = Constant()
  n2 = NamedRerouteUsage()
  n3 = VectorParameter() {group=CyberDelivery, parameterName=Logo_Offset, sortPriority=10, defaultValue=<rgba 0.276 -0.2219 0 0>}
  n4 = ComponentMask(None=@n3.RGB) {g=true, r=true}
  Add(A=AppendVector(A=Constant(), B=LinearInterpolate(B=@n1, Alpha=@n2)), B=@n4)
  Constant() {r=1}
  n5 = GetMaterialAttributes(None=FunctionInput() {inputName=MA, inputType=FunctionInput_MaterialAttributes}) {attributeGetTypes [#1]}
  n6 = Multiply(A=NamedRerouteUsage(), B=ScalarParameter() {defaultValue=0.5, group=CyberDelivery, parameterName=OriginalEmissive_Brightness}) {constB=0.5}
  n7 = Add(A=NamedRerouteUsage(), B=Add(A=NamedRerouteUsage(), B=NamedRerouteUsage()))
  n8 = OneMinus(None=@n7)
  n9 = VectorParameter() {group=CyberDelivery, parameterName=Digits_Color, defaultValue=<rgba 1 0.6725 0 1>}
  n10 = Add(
   A=Multiply(A=NamedRerouteUsage(), B=@n9.RGB)
   B=Multiply(A=Multiply(A=@n7, B=@n9.RGB), B=ScalarParameter() {defaultValue=2, group=CyberDelivery, parameterName=Digits_BrightnessMultiplier}) {constB=0.25}
  )
  n11 = VectorParameter() {group=CyberDelivery, parameterName=Pulse_Color, defaultValue=<rgba 1 0.3105 0 1>}
  n12 = Add(A=NamedRerouteUsage(), B=Add(A=NamedRerouteUsage(), B=NamedRerouteUsage()))
  n13 = Multiply(A=Multiply(A=Multiply(A=@n11.RGB, B=@n12)) {constB=0.75}, B=NamedRerouteUsage())
  n14 = NamedRerouteUsage()
  n15 = NamedRerouteUsage()
  n16 = Multiply(A=NamedRerouteUsage(), B=Add(A=Add(A=@n14, B=@n15), B=NamedRerouteUsage()))
  n17 = VectorParameter() {group=CyberDelivery, parameterName=Halftones_Color, defaultValue=<rgba 1 0.5363 0.197 1>}
  n18 = Multiply(A=Add(A=Add(A=@n10, B=@n13), B=Multiply(A=@n16, B=@n17.RGB)), B=NamedRerouteUsage())
  n19 = TextureCoordinate()
  n20 = Constant() {r=0.23}
  n21 = Constant() {r=0.12}
  n22 = Constant() {r=0.6}
  n23 = Constant() {r=0.85}
  n24 = Constant() {r=100}
  n25 = Saturate(None=Add(A=MaterialFunctionCall!(
     Texture Coordinate=@n19
     X Pos Right=@n20
     X Pos Left=@n21
     Y Pos Up=@n22
     Y Pos Down=@n23
     SquareSoftness=@n24
     RotateSquare
     RotationAmount
    ) {materialFunction=/Game/Accessories/FORT_Backpacks/Backpack_F_MED_GroovyReader_Play/Materials/MaterialFunctions/MF_SquareMaskGenerator}.SquareMask) {constB=0.1})
  n26 = Constant() {r=155}
  n27 = Add(A=MaterialFunctionCall!(ScanLine_Tiling_1=@n26, Scanline_PannerSpeed_1, ScanLine_Tiling=@n26, Scanline_PannerSpeed=Constant() {r=1.65}, Scanline_Rotation, UVs) {materialFunction=/Game/Global/Materials/MaterialGym/Materials/HolographicTexture/MF_HolographicTexture_ScanLines}.Result) {constB=0.23}
  n28 = Add(
   A=StaticSwitchParameter(True=Multiply(A=@n6, B=@n8), False=@n6) {group=CyberDelivery, parameterName=UseCrewStyleLogic}
   B=Multiply(A=Multiply(A=@n18, B=@n25), B=Saturate(None=@n27))
  )
  n29 = LinearInterpolate(Alpha=NamedRerouteUsage()) {constA=1, constAlpha=0, constB=1.85}
  FunctionOutput(None=SetMaterialAttributes!(MaterialAttributes=@n5.MaterialAttributes, EmissiveColor=Multiply(A=@n28, B=@n29)) {attributeSetTypes [#1]})
  n30 = ScalarParameter() {defaultValue=0.2, group=CyberDelivery, parameterName=Logo_Scale, sortPriority=10}
  n31 = NamedRerouteUsage()
  LinearInterpolate(A=@n30, B=Multiply(A=@n30) {constB=0.9}, Alpha=@n31) {constAlpha=1}
  n32 = VectorParameter() {parameterName=CelebrationColor_Lines, defaultValue=<rgba 0.2031 0.0556 0.0286 1>}
  n33 = Multiply(A=Time()) {constB=1.5}
  n34 = ComponentMask(None=Multiply(A=TextureCoordinate()) {constB=10}) {g=true}
  n35 = Multiply(A=@n34) {constB=2}
  n36 = Saturate(None=Sine(None=Add(A=@n33, B=@n35)))
  n37 = Add(A=TextureCoordinate(), B=ComponentMask(None=Constant2Vector() {g=-0.155, r=0.276}) {g=true, r=true})
  n38 = Constant() {r=0.21}
  n39 = TextureSample(UVs=Rotator(Coordinate=MaterialFunctionCall(UVs=@n37, Texture Scale=@n38) {materialFunction=&ScaleUVsByCenter}.UVs)) {
   samplerSource=SSM_Clamp_WorldGroupSettings
   samplerType=SAMPLERTYPE_LinearColor
   texture=/BRCosmetics/Accessories/FORT_Backpacks/Backpack_F_MED_CyberDelivery/Textures/T_F_MED_CyberDelivery_Backpack_CounterMasks
  }
  n40 = Add(A=Multiply(A=Time()) {constB=1.5}, B=Multiply(A=@n39.B) {constB=1.5})
  n41 = Constant() {r=0.85}
  n42 = Divide(A=Subtract(A=@n39.B, B=@n41), B=Subtract(A=Constant() {r=0.95}, B=@n41))
  n43 = Saturate(None=Subtract(A=@n39.B, B=Saturate(None=@n42)))
  n44 = Multiply(A=Saturate(None=Sine(None=@n40)), B=Saturate(None=Multiply(A=@n43) {constB=2}))
  n45 = Saturate(None=Subtract(A=Multiply(A=@n36, B=@n34), B=@n44))
  n46 = NamedRerouteUsage()
  n47 = Add(A=NamedRerouteUsage(), B=NamedRerouteUsage())
  n48 = Saturate(None=Subtract(A=@n44, B=Add(A=@n46, B=@n47)))
  n49 = VectorParameter() {parameterName=CelebrationColor_Bursts, defaultValue=<rgba 0.7726 0.3104 0 1>}
  Multiply(A=Add(A=Multiply(A=@n32.RGB, B=@n45), B=Multiply(A=@n48, B=@n49.RGB)), B=NamedRerouteUsage())
  n50 = Subtract(A=Time(), B=ScalarParameter() {defaultValue=9.999999980506448e+18, group=CyberDelivery, parameterName=BP_GetTIme, sortPriority=100})
  n51 = Saturate(None=Multiply(A=@n50) {constB=0.5})
  n52 = Constant() {r=1}
  NamedRerouteDeclaration(None=If(A=@n51, "A > B"=Floor(None=@n51), "A == B"=@n52, "A < B"=@n52)) {name=GlitchTiming, nodeColor=<rgba 1 0.3281 0 1>}
  n53 = NamedRerouteUsage()
  n54 = Multiply(A=NamedRerouteUsage()) {constB=-2}
  n55 = ScalarParameter() {defaultValue=-0.052, group=CyberDelivery, parameterName=Digits_Spacing, sortPriority=10}
  n56 = NamedRerouteUsage()
  n57 = AppendVector(A=Add(A=Add(A=@n54, B=@n55), B=LinearInterpolate(Alpha=@n56) {constB=-0.008}), B=Constant())
  n58 = MaterialFunctionCall(UVs=Add(A=Add(A=@n53, B=@n57), B=NamedRerouteUsage()), Texture Scale=NamedRerouteUsage()) {materialFunction=&ScaleUVsByCenter}
  n59 = Constant() {r=0.5}
  n60 = Floor(None=Divide(A=ComponentMask(None=@n58.UVs) {r=true}, B=Add(A=@n59) {constB=0.5}))
  n61 = ComponentMask(None=@n58.UVs) {r=true}
  n62 = Add(A=Constant() {r=0.5}) {constB=-0.5}
  n63 = Saturate(None=Floor(None=Divide(A=@n61, B=@n62)))
  n64 = Constant() {r=0.5}
  n65 = Floor(None=Divide(A=ComponentMask(None=@n58.UVs) {g=true}, B=Add(A=@n64) {constB=0.5}))
  n66 = ComponentMask(None=@n58.UVs) {g=true}
  n67 = Add(A=Constant() {r=0.5}) {constB=-0.5}
  n68 = Saturate(None=Floor(None=Divide(A=@n66, B=@n67)))
  n69 = Add(A=Add(A=Saturate(None=@n60), B=OneMinus(None=@n63)), B=Add(A=Saturate(None=@n65), B=OneMinus(None=@n68)))
  NamedRerouteDeclaration(None=OneMinus(None=Saturate(None=@n69))) {name=SecondDigit_Mask, nodeColor=<rgba 0.7109 0 1 1>}
  n70 = ScalarParameter() {group=CyberDelivery, parameterName=BP_FirstDigit}
  n71 = Constant() {r=4}
  n72 = Constant()
  n73 = NamedRerouteUsage()
  n74 = LinearInterpolate(A=@n72, B=Constant() {r=0.005}, Alpha=@n73)
  n75 = LinearInterpolate(A=@n72, B=Constant() {r=0.02}, Alpha=@n73)
  n76 = VectorParameter() {group=CyberDelivery, parameterName=Digits_Offset, sortPriority=10, defaultValue=<rgba 0.3 -0.3685 0 0>}
  n77 = ComponentMask(None=@n76.RGB) {g=true, r=true}
  n78 = AppendVector(A=NamedRerouteUsage(), B=Constant())
  n79 = Add(A=Add(A=AppendVector(A=@n74, B=@n75), B=Add(A=@n77, B=@n78)), B=TextureCoordinate())
  n80 = ScalarParameter() {defaultValue=0.12, group=CyberDelivery, parameterName=Digits_Scale, sortPriority=10}
  n81 = LinearInterpolate(A=@n80, B=Multiply(A=@n80) {constB=1.2}, Alpha=NamedRerouteUsage())
  n82 = Time()
  n83 = Multiply(A=Sine(None=Multiply(A=@n82) {constB=0.5}), B=NamedRerouteUsage())
  n84 = Abs(None=Multiply(A=@n83) {constB=0.04})
  n85 = MaterialFunctionCall(
   UVs=Add(A=NamedRerouteDeclaration(None=@n79) {name=NumbersUVsPos, nodeColor=<rgba 1 0 0.2812 1>}, B=NamedRerouteUsage())
   Texture Scale=NamedRerouteDeclaration(None=Add(A=@n81, B=@n84)) {name=NumbersUVsScale, nodeColor=<rgba 0 0.6016 1 1>}
  ) {materialFunction=&ScaleUVsByCenter}
  n86 = MaterialFunctionCall("Animation  Phase (0-1)"=Divide(A=Clamp(None=@n70) {maxDefault=9}) {constB=16}, Number of Rows=@n71, Number of Columns=@n71, UVs=@n85.UVs) {materialFunction=&FlipBook}
  n87 = MaterialFunctionCall!(UV=@n86.UVs, GlitchScale=Constant() {r=5}) {materialFunction=&/MF_HolographicTextureGlitch_CyberDelivery_Backpack}
  n88 = NamedRerouteUsage()
  n89 = TextureSample(UVs=LinearInterpolate(A=@n87.Glitched UV, B=@n86.UVs, Alpha=LinearInterpolate(Alpha=@n88) {constA=0.8}) {constAlpha=0.8}) {
   samplerSource=SSM_Wrap_WorldGroupSettings
   samplerType=SAMPLERTYPE_LinearColor
   texture=&T_F_MED_CyberDelivery_Backpack_CounterMasks
  }
  n90 = NamedRerouteUsage()
  n91 = Multiply(A=@n89.G, B=@n90)
  n92 = ScalarParameter() {defaultValue=0.5, group=CyberDelivery, parameterName=Digits_SDF_Min}
  n93 = Divide(A=Subtract(A=@n89.G, B=@n92), B=Subtract(A=ScalarParameter() {defaultValue=0.6, group=CyberDelivery, parameterName=Digits_SDF_Max}, B=@n92))
  n94 = NamedRerouteDeclaration(None=Multiply(A=@n90, B=Saturate(None=@n93))) {name=FirstDigit, nodeColor=<rgba 0 0.9766 1 1>}
  NamedRerouteDeclaration(None=Power(Base=Subtract(A=@n91, B=@n94), Exp=ScalarParameter() {defaultValue=1, group=CyberDelivery, parameterName=Digits_HalftonePower}) {constExponent=1.5}) {name=FirstDigit_HalftoneMask, nodeColor=<rgba 0.125 0 1 1>}
  n95 = ScalarParameter() {group=CyberDelivery, parameterName=BP_SecondDigit}
  n96 = Constant() {r=4}
  n97 = MaterialFunctionCall("Animation  Phase (0-1)"=Divide(A=Clamp(None=@n95) {maxDefault=9}) {constB=16}, Number of Rows=@n96, Number of Columns=@n96, UVs=@n58.UVs) {materialFunction=&FlipBook}
  n98 = MaterialFunctionCall(UV=@n97.UVs, GlitchScale=Constant() {r=5}) {materialFunction=&MF_HolographicTextureGlitch_CyberDelivery_Backpack}
  n99 = NamedRerouteUsage()
  n100 = TextureSample(UVs=LinearInterpolate(A=@n98.Glitched UV, B=@n97.UVs, Alpha=LinearInterpolate(Alpha=@n99) {constA=0.8}) {constAlpha=0.8}) {
   samplerSource=SSM_Wrap_WorldGroupSettings
   samplerType=SAMPLERTYPE_LinearColor
   texture=&T_F_MED_CyberDelivery_Backpack_CounterMasks
  }
  n101 = NamedRerouteUsage()
  n102 = Multiply(A=@n100.G, B=@n101)
  n103 = ScalarParameter() {defaultValue=0.5, group=CyberDelivery, parameterName=Digits_SDF_Min}
  n104 = Divide(
   A=Subtract(A=@n100.G, B=@n103)
   B=Subtract(A=ScalarParameter() {defaultValue=0.6, group=CyberDelivery, parameterName=Digits_SDF_Max}, B=@n103)
  )
  n105 = NamedRerouteDeclaration(None=Multiply(A=Saturate(None=@n104), B=@n101)) {name=SecondDigit, nodeColor=<rgba 1 0 0.6562 1>}
  NamedRerouteDeclaration(None=Power(Base=Subtract(A=@n102, B=@n105), Exp=ScalarParameter() {defaultValue=1, group=CyberDelivery, parameterName=Digits_HalftonePower}) {constExponent=1.5}) {name=SecondDigit_HalftoneMask, nodeColor=<rgba 0.1953 1 0 1>}
  n106 = Saturate(None=Multiply(A=@n50) {constB=0.3})
  n107 = OneMinus(None=Abs(None=Sine(None=@n106)))
  n108 = Multiply(A=@n107, B=@n107)
  NamedRerouteDeclaration(None=OneMinus(None=Multiply(A=@n108, B=@n108))) {name=GlowTiming, nodeColor=<rgba 0 0.4609 1 1>}
  n109 = LocalPosition() {localOrigin=InstancePreSkinning}
  n110 = SphereMask(
   A=ShaderStageSwitch(PixelShader=VertexInterpolator(VS=@n109.XYZ).PS, VertexShader=@n109.XYZ)
   B=Constant3Vector() {constant=<rgba 0 -16 0 0>}
  ) {attenuationRadius=10, hardnessPercent=0.2}
  n111 = ScalarParameter() {defaultValue=0.3, group=CyberDelivery, parameterName=SphereMaskGlow_Intensity}
  n112 = ScalarParameter() {defaultValue=0.3, group=CyberDelivery, parameterName=SphereMaskGlow_Intensity_OnKill}
  n113 = NamedRerouteUsage()
  NamedRerouteDeclaration(None=Multiply(A=Multiply(A=@n110, B=@n110), B=LinearInterpolate(A=@n111, B=@n112, Alpha=@n113)) {constB=0.1}) {name=SphereMask_Glow, nodeColor=<rgba 1 0.6328 0 1>}
  n114 = OneMinus(None=Saturate(None=Multiply(A=@n50) {constB=10}))
  n115 = Multiply(A=@n114, B=@n114)
  n116 = OneMinus(None=Multiply(A=@n115, B=@n115))
  n117 = Subtract(A=@n50, B=LinearInterpolate(Alpha=ScalarParameter() {parameterName=BP_IsFrontEnd}) {constA=4.1, constB=2.9}) {constB=4.1}
  n118 = OneMinus(None=Saturate(None=Multiply(A=@n117) {constB=2}))
  n119 = Multiply(A=@n118, B=@n118)
  n120 = OneMinus(None=@n118)
  n121 = Multiply(A=@n120, B=@n120)
  n122 = LinearInterpolate(A=Multiply(A=@n119, B=@n119), B=OneMinus(None=Multiply(A=@n121, B=@n121)), Alpha=@n118)
  NamedRerouteDeclaration(None=StaticSwitchParameter(True=Constant() {r=1}, False=Multiply(A=@n116, B=@n122)) {group=CyberDelivery, parameterName=DEBUG_OnKillAnimation}) {name=RepositionTiming, nodeColor=<rgba 0 0.4141 1 1>}
  n123 = Multiply(A=Time()) {constB=7}
  NamedRerouteDeclaration(None=Multiply(A=Sine(None=@n123)) {constB=0.0001}) {name=Flickering_UV, nodeColor=<rgba 0.1016 1 0 1>}
  n124 = Time()
  n125 = Multiply(A=Sine(None=Multiply(A=@n124) {constB=4})) {constB=0.5}
  NamedRerouteDeclaration(None=LinearInterpolate(Alpha=Add(A=@n125) {constB=0.5}) {constA=0.875}) {name=Flickering_Glow, nodeColor=<rgba 0 1 0.8828 1>}
  NamedRerouteDeclaration(None=Multiply(A=Abs(None=@n83)) {constB=0.01}) {name=Numbers_PositionOffsetX, nodeColor=<rgba 0.5938 0 1 1>}
  NamedRerouteDeclaration(None=@n5.EmissiveColor) {name=OriginalEmissive, nodeColor=<rgba 1 0 0.4922 1>}
  n126 = TextureCoordinate()
  n127 = ScalarParameter() {defaultValue=210, group=CyberDelivery, parameterName=Halftones_Scale}
  n128 = TextureSample(UVs=Rotator(Coordinate=Multiply(A=@n126, B=@n127) {constB=200}, Time=Constant() {r=0.3142})) {
   samplerSource=SSM_Wrap_WorldGroupSettings
   samplerType=SAMPLERTYPE_LinearGrayscale
   texture=/Game/Characters/Player/Female/Medium/Bodies/F_MED_TNTina/Materials/HalftoneOffset
  }
  NamedRerouteDeclaration(None=Saturate(None=Add(A=@n128.R) {constB=0})) {name=Halftone_Look, nodeColor=<rgba 0 0.8125 1 1>}
  n129 = Add(A=@n4, B=TextureCoordinate())
  n130 = NamedRerouteUsage()
  n131 = Sine(None=Multiply(A=Time()) {constB=2})
  n132 = Rotator(
   Coordinate=MaterialFunctionCall(UVs=Add(A=@n129, B=@n130), Texture Scale=@n30) {materialFunction=&ScaleUVsByCenter}.UVs
   Time=Multiply(A=@n31, B=Multiply(A=@n131) {constB=0.4})
  )
  n133 = MaterialFunctionCall(UV=@n132) {materialFunction=&MF_HolographicTextureGlitch_CyberDelivery_Backpack}
  n134 = TextureSample(UVs=LinearInterpolate(A=@n133.Glitched UV, B=@n132, Alpha=NamedRerouteUsage())) {
   samplerSource=SSM_Clamp_WorldGroupSettings
   samplerType=SAMPLERTYPE_LinearColor
   texture=&T_F_MED_CyberDelivery_Backpack_CounterMasks
  }
  n135 = LinearInterpolate(A=@n134.R, B=@n134.R, Alpha=Round(None=NamedRerouteUsage()))
  n136 = Sine(None=Add(A=Time(), B=Multiply(A=@n135) {constB=1.5}))
  n137 = Multiply(A=@n135) {constB=2}
  NamedRerouteDeclaration(None=Multiply(A=Saturate(None=@n136), B=Saturate(None=@n137))) {name=Cat_Pulse, nodeColor=<rgba 0.3828 0 1 1>}
  n138 = Add(A=Time(), B=Multiply(A=@n89.G) {constB=2})
  n139 = Add(A=Multiply(A=Sine(None=@n138)) {constB=0.5}) {constB=0.5}
  NamedRerouteDeclaration(None=Multiply(A=Multiply(A=@n139, B=@n89.G), B=NamedRerouteUsage())) {name=FirstDigit_Pulse, nodeColor=<rgba 1 0 0.3516 1>}
  n140 = Add(A=Time(), B=Multiply(A=@n100.G) {constB=2})
  n141 = Add(A=Multiply(A=Sine(None=@n140)) {constB=0.5}) {constB=0.5}
  NamedRerouteDeclaration(None=Multiply(A=Multiply(A=@n141, B=@n100.G), B=NamedRerouteUsage())) {name=SecondDigit_Pulse, nodeColor=<rgba 0 0.4375 1 1>}
  n142 = ScalarParameter() {defaultValue=0.45, group=CyberDelivery, parameterName=Cat_SDF_Min}
  n143 = Subtract(A=@n135, B=@n142)
  n144 = Subtract(A=ScalarParameter() {defaultValue=0.6, group=CyberDelivery, parameterName=Cat_SDF_Max}, B=@n142)
  n145 = NamedRerouteDeclaration(None=Saturate(None=Divide(A=@n143, B=@n144))) {name=CatLogo, nodeColor=<rgba 1 0.7031 0 1>}
  NamedRerouteDeclaration(None=Power(Base=Subtract(A=@n135, B=@n145)) {constExponent=0.75}) {name=Cat_HalftoneMask, nodeColor=<rgba 0 1 0.2266 1>}
  n146 = Constant() {r=0.5}
  n147 = Floor(None=Divide(A=ComponentMask(None=@n85.UVs) {r=true}, B=Add(A=@n146) {constB=0.5}))
  n148 = ComponentMask(None=@n85.UVs) {r=true}
  n149 = Add(A=Constant() {r=0.5}) {constB=-0.5}
  n150 = Saturate(None=Floor(None=Divide(A=@n148, B=@n149)))
  n151 = Constant() {r=0.5}
  n152 = Floor(None=Divide(A=ComponentMask(None=@n85.UVs) {g=true}, B=Add(A=@n151) {constB=0.5}))
  n153 = ComponentMask(None=@n85.UVs) {g=true}
  n154 = Add(A=Constant() {r=0.5}) {constB=-0.5}
  n155 = Saturate(None=Floor(None=Divide(A=@n153, B=@n154)))
  n156 = Add(A=Add(A=Saturate(None=@n147), B=OneMinus(None=@n150)), B=Add(A=Saturate(None=@n152), B=OneMinus(None=@n155)))
  NamedRerouteDeclaration(None=OneMinus(None=Saturate(None=@n156))) {name=FirstDigit_Mask, nodeColor=<rgba 1 0 0.7734 1>}
 ]}
