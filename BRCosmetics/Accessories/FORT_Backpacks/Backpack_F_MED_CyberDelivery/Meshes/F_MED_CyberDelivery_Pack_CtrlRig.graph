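# Control Rig for the CyberDelivery backpack cosmetic that rotates its two sword bones as the reference spine bends.
# Forward solve remaps the local Z Euler angle of five spine reference bones and sums them through a collapsed Add node.
# That sum weights mirrored fixed relative rotations on the left and right sword bones.
# A variable-named bone, helper control and colored XYZ prints form a separate rotation-debugging setup.
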ControlRigBlueprint /BRCosmetics/Accessories/FORT_Backpacks/Backpack_F_MED_CyberDelivery/Meshes/F_MED_CyberDelivery_Pack_CtrlRig
graphs [
 Add_1_ContainedGraph {nodes [
   n1 = RigVMFunctionEntryNode()
   Add = &RigVMFunction_MathDoubleAdd(A=@n1.A, B=@n1.B)
   n2 = &RigVMFunction_MathDoubleAdd(A=@Add.Result, B=@n1.C)
   RigVMFunctionReturnNode(Result=&RigVMFunction_MathDoubleAdd(A=&RigVMFunction_MathDoubleAdd(A=@n2.Result, B=@n1.D).Result, B=@n1.E).Result)
  ]}
 RigVMFunctionLibrary {links [], nodes []}
 RigVMModel {nodes [
   n1 = &RigUnit_PrepareForExecution()
   RigUnit_GetTransform = &RigUnit_GetTransform(Item=(Type=Bone,Name="lowerarm_l"), Item.Name=RigVMVariableNode() {variable=NewControl}.Value)
   SetControlOffset = /Script/ControlRig.RigUnit_SetControlOffset(Control=NewControl, Space=GlobalSpace, ExecutePin=@n1.ExecutePin, Offset=@RigUnit_GetTransform.Transform) {is=struct RigVMUnitNode, template="SetControlOffset::Execute(in Control,io Offset,in Space)"}
   RigUnit_BeginExecution = &RigUnit_BeginExecution()
   n2 = &RigUnit_GetTransform(Item=(Type=Bone,Name="spine_02_ref"), Space=LocalSpace, bInitial=False)
   n3 = &RigUnit_GetTransform(Item=(Type=Bone,Name="spine_03_ref"), Space=LocalSpace, bInitial=False)
   n4 = &RigUnit_GetTransform(Item=(Type=Bone,Name="spine_01_ref"), Space=LocalSpace, bInitial=False)
   n5 = &RigUnit_GetTransform(Item=(Type=Bone,Name="spine_04_ref"), Space=LocalSpace, bInitial=False)
   n6 = &RigUnit_GetTransform(Item=(Type=Bone,Name="spine_05_ref"), Space=LocalSpace, bInitial=False)
   Add = RigVMAggregateNode(
    B=&RigVMFunction_MathDoubleRemap(
     SourceMaximum=-15
     SourceMinimum=-5
     TargetMaximum=0.2
     TargetMinimum=0
     Value=&RigVMFunction_MathQuaternionToEuler(RotationOrder=YXZ, Value=@n2.Transform.Rotation).Result.Z
    ).Result
    C=&RigVMFunction_MathDoubleRemap(
     SourceMaximum=-14
     SourceMinimum=-4
     TargetMaximum=0.2
     TargetMinimum=0
     Value=&RigVMFunction_MathQuaternionToEuler(RotationOrder=YXZ, Value=@n3.Transform.Rotation).Result.Z
    ).Result
    A=&RigVMFunction_MathDoubleRemap(
     SourceMaximum=-16
     SourceMinimum=-6
     TargetMaximum=0.2
     TargetMinimum=0
     Value=&RigVMFunction_MathQuaternionToEuler(RotationOrder=YXZ, Value=@n4.Transform.Rotation).Result.Z
    ).Result
    D=&RigVMFunction_MathDoubleRemap(
     SourceMaximum=-10
     SourceMinimum=0
     TargetMaximum=0.2
     TargetMinimum=0
     Value=&RigVMFunction_MathQuaternionToEuler(RotationOrder=YXZ, Value=@n5.Transform.Rotation).Result.Z
    ).Result
    E=&RigVMFunction_MathDoubleRemap(
     SourceMaximum=-12
     SourceMinimum=-2
     TargetMaximum=0.2
     TargetMinimum=0
     Value=&RigVMFunction_MathQuaternionToEuler(RotationOrder=YXZ, Value=@n6.Transform.Rotation).Result.Z
    ).Result
   )
   /Script/ControlRig.RigUnit_SetRelativeTransformForItem(
    Child=(Type=Bone,Name="dyn_sword_01_l")
    Parent=(Type=Bone,Name="dyn_sword_01_l")
    Value=(Rotation=(X=0.5152,Y=-0.1526,Z=-0.2395,W=0.8087),Translation=(X=0,Y=0,Z=0),Scale3D=(X=1,Y=1,Z=1))
    bParentInitial=false
    bPropagateToChildren=true
    Weight=@Add.Result
    ExecutePin=@RigUnit_BeginExecution.ExecutePin
   ) {
    is=struct RigVMUnitNode
    template="Set Relative Transform::Execute(in Child,in Parent,in bParentInitial,in Value,in Weight,in bPropagateToChildren)"
   }
   > &RigUnit_SetRelativeTransformForItem(
    Child=(Type=Bone,Name="dyn_sword_01_r")
    Parent=(Type=Bone,Name="dyn_sword_01_r")
    Value=(Rotation=(X=-0.5152,Y=0.1526,Z=-0.2395,W=0.8087),Translation=(X=0,Y=0,Z=0),Scale3D=(X=1,Y=1,Z=1))
    bParentInitial=false
    bPropagateToChildren=true
    Weight=@Add.Result
   ) {template="Set Relative Transform::Execute(in Child,in Parent,in bParentInitial,in Value,in Weight,in bPropagateToChildren)"}
   RigUnit_SetTransform = &RigUnit_SetTransform(
    ExecutePin=()
    Item=(Type=Bone,Name="lowerarm_l")
    Weight=1
    bInitial=False
    Value=&RigUnit_GetTransform(Item=(Type=Control,Name="NewControl")).Transform
    Item.Name=RigVMVariableNode() {variable=NewControl}.Value
   )
   n7 = &RigVMFunction_MathQuaternionToEuler(
    RotationOrder=YXZ
    Value=&RigUnit_GetTransform(Item=(Type=Bone,Name="deform_glute_r"), Space=LocalSpace, Item.Name=RigVMVariableNode() {variable=NewControl}.Value).Transform.Rotation
   )
   > DISPATCH_RigVMDispatch_Print = RigVMDispatchNode(Enabled=True, Prefix="X: ", ScreenColor=(R=1,G=0,B=0.0348,A=1), ScreenDuration=0.05, Value=@n7.Result.X) {template="DISPATCH_RigVMDispatch_Print(in Prefix,in Value,in Enabled,in ScreenDuration,in ScreenColor)"}
   > RigVMDispatchNode(Enabled=True, Prefix="Y: ", ScreenColor=(R=0.0533,G=1,B=0.0718,A=1), ScreenDuration=0.05, Value=@n7.Result.Y) {template="DISPATCH_RigVMDispatch_Print(in Prefix,in Value,in Enabled,in ScreenDuration,in ScreenColor)"}
   > RigVMDispatchNode(Enabled=True, Prefix="Z: ", ScreenColor=(R=0.5035,G=0.5393,B=1,A=1), ScreenDuration=0.05, Value=@n7.Result.Z) {template="DISPATCH_RigVMDispatch_Print(in Prefix,in Value,in Enabled,in ScreenDuration,in ScreenColor)"}
  ]}
]
hierarchy [
 root {loc=<0 0 0>, rot=<0 0 0 1>}
 pack {parent=root, loc=<0 -0.7316 0>, rot=<0 0 0 1>}
 dyn_strap_BASE_l {parent=pack, loc=<4.1286 -0.526 -10.6655>, rot=<-0.2024 -0.6843 -0.2011 -0.6711>}
 dyn_strap_01_l {parent=dyn_strap_BASE_l, loc=<0 0 0>, rot=<0 0 0 1>}
 dyn_strap_02_l {parent=dyn_strap_01_l, loc=<2.3474 -0.5624 -0.051>, rot=<-0.008 -0.0349 0.3896 -0.9203>}
 dyn_strap_03_l {parent=dyn_strap_02_l, loc=<2.023 -0.5096 -0.1004>, rot=<0.0017 -0.0195 -0.0107 -0.9998>}
 dyn_strap_04_l {parent=dyn_strap_03_l, loc=<2.6331 0.8059 0.5698>, rot=<-0.0146 0.2004 -0.0966 -0.9748>}
 dyn_strap_BASE_r {parent=pack, loc=<-4.174 -0.4046 -10.6225>, rot=<-0.2743 -0.6531 -0.2694 -0.6524>}
 dyn_strap_01_r {parent=dyn_strap_BASE_r, loc=<0 0 0>, rot=<0 0 0 1>}
 dyn_strap_02_r {parent=dyn_strap_01_r, loc=<2.6729 -0.8773 0.0862>, rot=<0.0182 0.0681 0.4804 -0.8742>}
 dyn_strap_03_r {parent=dyn_strap_02_r, loc=<2.7284 -0.6835 0.1806>, rot=<0.0106 0.0076 0.0626 -0.998>}
 dyn_strap_04_r {parent=dyn_strap_03_r, loc=<2.3108 0.7774 -0.2738>, rot=<0.0102 -0.0778 -0.1563 -0.9846>}
 pelvis_ref {parent=pack, loc=<0 17.207 -40.1139>, rot=<0.0127 0.707 -0.0127 -0.707>}
 spine_01_ref {parent=pelvis_ref, loc=<6.7494 -0.6388 0>, rot=<0 0 0.034 -0.9994>}
 spine_02_ref {parent=spine_01_ref, loc=<7.2518 -0.035 0>, rot=<0 0 0.0238 -0.9997>}
 spine_03_ref {parent=spine_02_ref, loc=<6.0104 0 0>, rot=<0 0 0.0149 -0.9999>}
 spine_04_ref {parent=spine_03_ref, loc=<6.1005 0 0>, rot=<0 0 0 1>}
 spine_05_ref {parent=spine_04_ref, loc=<5.9951 0 0>, rot=<0 0 0 1>}
 dyn_sword_BASE_l {parent=pack, loc=<7.2087 -9.3837 -10.3047>, rot=<-0.0116 -0.7652 0.1115 -0.634>}
 dyn_sword_01_l {parent=dyn_sword_BASE_l, loc=<0 0 0>, rot=<0 0 0 -1>}
 sword_hide_l {parent=dyn_sword_01_l, loc=<-20.7641 -0.0142 3.0394>, rot=<0 0 1 0>}
 dyn_sword_BASE_r {parent=pack, loc=<-7.2087 -9.3837 -10.3047>, rot=<0.1115 -0.634 -0.0116 -0.7652>}
 dyn_sword_01_r {parent=dyn_sword_BASE_r, loc=<0 0 0>, rot=<0 0 0 -1>}
 sword_hide_r {parent=dyn_sword_01_r, loc=<-20.7642 -0.0142 -3.0394>, rot=<0 0 0 -1>}
 NewControl {controlType=EULER_TRANSFORM, parent=root, type=Control, loc=<0 0 0>, rot=<0 0 0 1>}
]
variables [NewControl {type=FName, default=spine_01_ref}]
