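# Physics asset for the CyberDelivery backpack cosmetic, defining capsule bodies for the pack, two dangling straps and pelvis/spine reference bones.
# Each strap chains four links from a base on the pack with limited swing and twist, the last link bending almost solely on one swing axis.
# Strap bases and reference bones join with free angular constraints, the reference capsules likely standing in for the wearer's torso as colliders.
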PhysicsAsset /BRCosmetics/Accessories/FORT_Backpacks/Backpack_F_MED_CyberDelivery/Meshes/F_MED_CyberDelivery_Pack_Physics
bodies {
 dyn_strap_01_l [{shapeType=Capsule, length=2.6747, radius=0.7253, shapeName=dyn_strap_01_l_capsule, center=<1 0 0>, rotation=<rot 90 0 0>}]
 dyn_strap_01_r [{shapeType=Capsule, length=2.6747, radius=0.9253, shapeName=dyn_strap_01_r_capsule, center=<1 0 0>, rotation=<rot 90 0 0>}]
 dyn_strap_02_l [{shapeType=Capsule, length=2.6747, radius=0.7253, shapeName=dyn_strap_02_l_capsule, center=<1 0 0>, rotation=<rot 90 0 0>}]
 dyn_strap_02_r [{shapeType=Capsule, length=2.6747, radius=0.9253, shapeName=dyn_strap_02_r_capsule, center=<1 0 0>, rotation=<rot 90 0 0>}]
 dyn_strap_03_l [{shapeType=Capsule, length=2.6747, radius=0.7253, shapeName=dyn_strap_03_l_capsule, center=<1 0 0>, rotation=<rot 90 0 0>}]
 dyn_strap_03_r [{shapeType=Capsule, length=2.6747, radius=0.9253, shapeName=dyn_strap_03_r_capsule, center=<1 0 0>, rotation=<rot 90 0 0>}]
 dyn_strap_04_l [{shapeType=Capsule, length=2.6747, radius=0.7253, shapeName=dyn_strap_04_l_capsule, center=<1 0 0>, rotation=<rot 90 0 0>}]
 dyn_strap_04_r [{shapeType=Capsule, length=2.6747, radius=0.9253, shapeName=dyn_strap_04_r_capsule, center=<2 0 0>, rotation=<rot 90 0 0>}]
 dyn_strap_BASE_l [{shapeType=Capsule, length=2, radius=0.8, shapeName=dyn_strap_BASE_l_capsule}]
 dyn_strap_BASE_r [{shapeType=Capsule, length=2, radius=0.8, shapeName=dyn_strap_BASE_r_capsule}]
 pack [{shapeType=Capsule, length=9, radius=12, shapeName=pack_capsule, center=<0 -8.1233 -0.0059>, rotation=<rot 0 0 23.8692>}]
 pelvis_ref [{shapeType=Capsule, length=11, radius=16, shapeName=pelvis_ref_capsule, rotation=<rot 0 -90 0>}]
 spine_01_ref [{shapeType=Capsule, length=8.8039, radius=11.6961, shapeName=spine_01_ref_capsule, rotation=<rot 0 -90 0>}]
 spine_02_ref [{shapeType=Capsule, length=6.3039, radius=10.6961, shapeName=spine_02_ref_capsule, rotation=<rot 0 -90 0>}]
 spine_03_ref [{shapeType=Capsule, length=5.3039, radius=9.6961, shapeName=spine_03_ref_capsule, rotation=<rot 0 -90 0>}]
}
constraints [
 {bone1Name=dyn_strap_01_l, bone2Name=dyn_strap_BASE_l, swing1LimitDegrees=38.64, swing2LimitDegrees=10.5, twistLimitDegrees=6}
 {bone1Name=dyn_strap_02_l, bone2Name=dyn_strap_01_l, swing1LimitDegrees=41.4, swing2LimitDegrees=17, twistLimitDegrees=20}
 {bone1Name=dyn_strap_03_l, bone2Name=dyn_strap_02_l, swing1LimitDegrees=46, swing2LimitDegrees=21.7, twistLimitDegrees=21}
 {bone1Name=dyn_strap_04_l, bone2Name=dyn_strap_03_l, swing1LimitDegrees=85, swing2LimitDegrees=0.1, twistLimitDegrees=0.1}
 {bone1Name=dyn_strap_01_r, bone2Name=dyn_strap_BASE_r, swing1LimitDegrees=36.12, swing2LimitDegrees=12.6, twistLimitDegrees=4}
 {bone1Name=dyn_strap_02_r, bone2Name=dyn_strap_01_r, swing1LimitDegrees=38.7, swing2LimitDegrees=15, twistLimitDegrees=18}
 {bone1Name=dyn_strap_03_r, bone2Name=dyn_strap_02_r, swing1LimitDegrees=43, swing2LimitDegrees=19.6, twistLimitDegrees=19}
 {bone1Name=dyn_strap_04_r, bone2Name=dyn_strap_03_r, swing1LimitDegrees=75, swing2LimitDegrees=0.1, twistLimitDegrees=0.1}
 {
  bone1Name=dyn_strap_BASE_l
  bone2Name=pack
  swing1LimitDegrees=45
  swing1Motion=Free
  swing2LimitDegrees=45
  swing2Motion=Free
  twistLimitDegrees=45
  twistMotion=Free
 }
 {
  bone1Name=dyn_strap_BASE_r
  bone2Name=pack
  swing1LimitDegrees=45
  swing1Motion=Free
  swing2LimitDegrees=45
  swing2Motion=Free
  twistLimitDegrees=45
  twistMotion=Free
 }
 {
  bone1Name=pelvis_ref
  bone2Name=pack
  swing1LimitDegrees=45
  swing1Motion=Free
  swing2LimitDegrees=45
  swing2Motion=Free
  twistLimitDegrees=45
  twistMotion=Free
 }
 {
  bone1Name=spine_01_ref
  bone2Name=pelvis_ref
  swing1LimitDegrees=45
  swing1Motion=Free
  swing2LimitDegrees=45
  swing2Motion=Free
  twistLimitDegrees=45
  twistMotion=Free
 }
 {
  bone1Name=spine_02_ref
  bone2Name=spine_01_ref
  swing1LimitDegrees=45
  swing1Motion=Free
  swing2LimitDegrees=45
  swing2Motion=Free
  twistLimitDegrees=45
  twistMotion=Free
 }
 {
  bone1Name=spine_03_ref
  bone2Name=spine_02_ref
  swing1LimitDegrees=45
  swing1Motion=Free
  swing2LimitDegrees=45
  swing2Motion=Free
  twistLimitDegrees=45
  twistMotion=Free
 }
]
