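# Physics asset for the sword variant of the CyberDelivery backpack cosmetic, defining simulated motion for its left and right swords.
# Each sword bone has a long thin capsule body, and each matching base bone has a tiny capsule that serves as its joint partner.
# Two mirrored constraints link each sword to its base with narrow swing and twist limits, so the swords can sway only slightly.
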PhysicsAsset /BRCosmetics/Accessories/FORT_Backpacks/Backpack_F_MED_CyberDelivery/Meshes/F_MED_CyberDelivery_Pack_Physics_Swords
bodies {
 dyn_sword_01_l [{shapeType=Capsule, length=45, radius=2.3, shapeName=dyn_sword_01_l_capsule, center=<-0.6043 0.981 1.8315>, rotation=<rot 0 90.047 -89.8579>}]
 dyn_sword_01_r [{
   shapeType=Capsule
   length=45
   radius=2.3
   shapeName=dyn_sword_01_r_capsule
   center=<-0.5273 1.1212 -2.8908>
   rotation=<rot 0 -89.953 -89.858>
  }]
 dyn_sword_BASE_l [{shapeType=Capsule, length=0.505, radius=0.505, shapeName=dyn_sword_BASE_l_capsule, rotation=<rot 0 -90 0>}]
 dyn_sword_BASE_r [{shapeType=Capsule, length=0.505, radius=0.505, shapeName=dyn_sword_BASE_r_capsule, rotation=<rot 0 -90 0>}]
}
constraints [
 {bone1Name=dyn_sword_01_r, bone2Name=dyn_sword_BASE_r, swing1LimitDegrees=25, swing2LimitDegrees=10, twistLimitDegrees=10}
 {bone1Name=dyn_sword_01_l, bone2Name=dyn_sword_BASE_l, swing1LimitDegrees=25, swing2LimitDegrees=10, twistLimitDegrees=10}
]
