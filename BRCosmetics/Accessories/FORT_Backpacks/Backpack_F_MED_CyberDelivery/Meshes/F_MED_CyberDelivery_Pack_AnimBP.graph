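# Animation Blueprint for the CyberDelivery backpack cosmetic, adding secondary motion to the pack on the character's back.
# Each update copies the parent mesh's pelvis and spine socket transforms onto reference bones, cached as simulation input.
# bRigidBodyEnabled selects gender-specific RigidBody simulation, chosen from the customization part, or AnimDynamics strap chains.
# The result passes through a swords RigidBody simulation, then a Control Rig.
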AnimBlueprint /BRCosmetics/Accessories/FORT_Backpacks/Backpack_F_MED_CyberDelivery/Meshes/F_MED_CyberDelivery_Pack_AnimBP
target_skeleton=&/F_MED_CyberDelivery_Pack_Skeleton
graphs [
 AnimGraph {nodes [
   n1 = RigidBody!(ComponentPose=Local To Component(LocalPose=Use cached pose 'InputCache'()), Alpha, OverrideWorldGravity:Vector="0,0,0", ExternalForce) -> (Pose) {
    bClampLinearTranslationLimitToRefPose=true
    overridePhysicsAsset=&/F_MED_CyberDelivery_Pack_Physics_Male
    simulationSpace=BaseBoneSpace
    alphaBoolBlend {blendOption=Linear}
    baseBoneRef {boneName=pack}
    simSpaceSettings {
     maxAngularAcceleration=24
     maxAngularVelocity=8
     maxLinearAcceleration=2100
     maxLinearVelocity=700
     velocityScaleZ=0.7
     worldAlpha=1
     externalLinearDragV=<2 2 2>
    }
   }
   n2 = RigidBody(ComponentPose=Local To Component(LocalPose=Use cached pose 'InputCache'()), OverrideWorldGravity="0,0,0") {
    bClampLinearTranslationLimitToRefPose=true
    overridePhysicsAsset=&/F_MED_CyberDelivery_Pack_Physics
    simulationSpace=BaseBoneSpace
    alphaBoolBlend {blendOption=Linear}
    baseBoneRef {boneName=pack}
    simSpaceSettings {
     maxAngularAcceleration=24
     maxAngularVelocity=8
     maxLinearAcceleration=2100
     maxLinearVelocity=700
     velocityScaleZ=0.7
     worldAlpha=1
     externalLinearDragV=<2 2 2>
    }
   }
   n3 = Blend Poses by bool(
    BlendPose_0=Component To Local(ComponentPose=@n2)
    BlendPose_1=Component To Local(ComponentPose=@n1)
    bActiveValue=Get isF_Med() -> (isF_Med:bool)
   ) {blendType=Linear}
   n4 = "Anim Dynamics - Chain: dyn_strap_01_l -> dyn_strap_03_l"!(ComponentPose=Local To Component(LocalPose=Use cached pose 'InputCache'()), Alpha=1, GravityOverride:Vector="0,0,0", ExternalForce="0,0,0") -> (Pose) {
    angularBiasOverride=1
    angularDampingOverride=1
    bChain=true
    bOverrideAngularBias=true
    bOverrideAngularDamping=true
    bOverrideLinearDamping=true
    bUseGravityOverride=true
    linearDampingOverride=1
    simulationSpace=RootRelative
    alphaBoolBlend {blendOption=Linear}
    alphaScaleBiasClamp {scale=0.45}
    boundBone {boneName=dyn_strap_01_l}
    chainEnd {boneName=dyn_strap_03_l}
    componentAppliedLinearAccClamp=<1000 1000 1000>
    componentLinearAccScale=<0.7 0.7 0.5>
    componentLinearVelScale=<0.7 0.7 0.5>
    physicsBodyDefinitions [
     {
      boundBone {boneName=dyn_strap_01_l}
      boxExtents=<3.6 1.8 7>
      constraintSetup {angularLimitsMax=<9 14 4>, angularLimitsMin=<-9 -14 -19>, angularTarget=<0 0 0>}
      localJointOffset=<0 0 0>
     }
     {
      boundBone {boneName=dyn_strap_02_l}
      boxExtents=<3.6 1.8 7>
      constraintSetup {angularLimitsMax=<14 19 9>, angularLimitsMin=<-14 -19 -47>, angularTarget=<0 0 0>}
      localJointOffset=<0 0 0>
     }
     {
      boundBone {boneName=dyn_strap_03_l}
      boxExtents=<3.6 1.8 7>
      constraintSetup {angularLimitsMax=<19 32 9>, angularLimitsMin=<-19 -23 -67>, angularTarget=<0 0 0>}
      localJointOffset=<0 0 0>
     }
    ]
   }
   n5 = "Anim Dynamics - Chain: dyn_strap_01_r -> dyn_strap_04_r"!(ComponentPose=@n4, Alpha=1, GravityOverride:Vector="0,0,0", ExternalForce="0,0,0") -> (Pose) {
    angularBiasOverride=1
    angularDampingOverride=1
    bChain=true
    bOverrideAngularBias=true
    bOverrideAngularDamping=true
    bOverrideLinearDamping=true
    bUseGravityOverride=true
    linearDampingOverride=1
    simulationSpace=RootRelative
    alphaBoolBlend {blendOption=Linear}
    alphaScaleBiasClamp {scale=0.57}
    boundBone {boneName=dyn_strap_01_r}
    chainEnd {boneName=dyn_strap_04_r}
    componentAppliedLinearAccClamp=<1000 1000 1000>
    componentLinearAccScale=<0.85 0.85 0.65>
    componentLinearVelScale=<0.85 0.85 0.65>
    physicsBodyDefinitions [
     {
      boundBone {boneName=dyn_strap_01_r}
      boxExtents=<4.4 2.2 8.8>
      constraintSetup {angularLimitsMax=<11 16 5>, angularLimitsMin=<-11 -16 -26>, angularTarget=<0 0 0>}
      localJointOffset=<0.1 0 0>
     }
     {
      boundBone {boneName=dyn_strap_02_r}
      boxExtents=<4.4 2.2 8.8>
      constraintSetup {angularLimitsMax=<16 23 11>, angularLimitsMin=<-16 -22 -52>, angularTarget=<0 0 0>}
      localJointOffset=<0.1 0 0>
     }
     {
      boundBone {boneName=dyn_strap_03_r}
      boxExtents=<4.4 2.2 8.8>
      constraintSetup {angularLimitsMax=<21 27 11>, angularLimitsMin=<-22 -26 -72>, angularTarget=<0 0 0>}
      localJointOffset=<0.1 0 0>
     }
     {
      boundBone {boneName=dyn_strap_04_r}
      boxExtents=<4.4 2.2 8.8>
      constraintSetup {angularLimitsMax=<0.1 0.1 100>, angularLimitsMin=<-0.1 -0.1 -100>, angularTarget=<0 0 0>}
      localJointOffset=<0.1 0 0>
     }
    ]
   }
   n6 = Component To Local(ComponentPose=@n5)
   n7 = Get bRigidBodyEnabled()
   n8 = RigidBody(
    ComponentPose=Local To Component(LocalPose=Blend Poses by bool(BlendPose_0=@n3, BlendPose_1=@n6, bActiveValue=@n7) {blendType=Linear})
    OverrideWorldGravity="0,0,0"
   ) {
    bClampLinearTranslationLimitToRefPose=true
    overridePhysicsAsset=&/F_MED_CyberDelivery_Pack_Physics_Swords
    simulationSpace=BaseBoneSpace
    alphaBoolBlend {blendOption=Linear}
    baseBoneRef {boneName=pack}
    simSpaceSettings {
     maxAngularAcceleration=24
     maxAngularVelocity=8
     maxLinearAcceleration=2100
     maxLinearVelocity=700
     velocityScaleZ=0.8
     worldAlpha=1
     externalLinearDragV=<0.5 0.5>
    }
   }
   n9 = Control Rig(Source=Component To Local(ComponentPose=@n8)) {
    alphaBoolBlend {blendOption=Linear}
    controlRigAssetReference {blueprintRigClass=&/F_MED_CyberDelivery_Pack_CtrlRig.F_MED_CyberDelivery_Pack_CtrlRig_C}
   }
   Output Pose(Result=@n9) {layerGroup=None}
   n10 = "Transform (Modify) Bone - Bone: pelvis_ref"(
    ComponentPose=Local To Component(LocalPose=Input Pose())
    Alpha=1
    Translation=Get pelvis_01_tran() -> (pelvis_01_tran:Vector)
    Rotation=Get pelvis_01_rot() -> (pelvis_01_rot:Rotator)
    Scale="1,1,1"
   ) {
    rotationMode=BMM_Replace
    rotationSpace=BCS_WorldSpace
    scaleSpace=BCS_BoneSpace
    translationMode=BMM_Replace
    translationSpace=BCS_WorldSpace
    alphaBoolBlend {blendOption=Linear}
    alphaScaleBiasClamp {inRange {max=0, min=-40}, outRange {max=0, min=1}}
    boneToModify {boneName=pelvis_ref}
   }
   n11 = "Transform (Modify) Bone - Bone: spine_01_ref"(
    ComponentPose=@n10
    Alpha=1
    Translation=Get spine_01_tran() -> (spine_01_tran:Vector)
    Rotation=Get spine_01_rot() -> (spine_01_rot:Rotator)
    Scale="1,1,1"
   ) {
    rotationMode=BMM_Replace
    rotationSpace=BCS_WorldSpace
    scaleSpace=BCS_BoneSpace
    translationMode=BMM_Replace
    translationSpace=BCS_WorldSpace
    alphaBoolBlend {blendOption=Linear}
    alphaScaleBiasClamp {inRange {max=0, min=-40}, outRange {max=0, min=1}}
    boneToModify {boneName=spine_01_ref}
   }
   n12 = "Transform (Modify) Bone - Bone: spine_02_ref"(
    ComponentPose=@n11
    Alpha=1
    Translation=Get spine_02_tran() -> (spine_02_tran:Vector)
    Rotation=Get spine_02_rot() -> (spine_02_rot:Rotator)
    Scale="1,1,1"
   ) {
    rotationMode=BMM_Replace
    rotationSpace=BCS_WorldSpace
    scaleSpace=BCS_BoneSpace
    translationMode=BMM_Replace
    translationSpace=BCS_WorldSpace
    alphaBoolBlend {blendOption=Linear}
    alphaScaleBiasClamp {inRange {max=0, min=-40}, outRange {max=0, min=1}}
    boneToModify {boneName=spine_02_ref}
   }
   n13 = "Transform (Modify) Bone - Bone: spine_03_ref"(
    ComponentPose=@n12
    Alpha=1
    Translation=Get spine_03_tran() -> (spine_03_tran:Vector)
    Rotation=Get spine_03_rot() -> (spine_03_rot:Rotator)
    Scale="1,1,1"
   ) {
    rotationMode=BMM_Replace
    rotationSpace=BCS_WorldSpace
    scaleSpace=BCS_BoneSpace
    translationMode=BMM_Replace
    translationSpace=BCS_WorldSpace
    alphaBoolBlend {blendOption=Linear}
    alphaScaleBiasClamp {inRange {max=0, min=-40}, outRange {max=0, min=1}}
    boneToModify {boneName=spine_03_ref}
   }
   n14 = "Transform (Modify) Bone - Bone: spine_04_ref"(
    ComponentPose=@n13
    Alpha=1
    Translation=Get pelvis_04_tran() -> (pelvis_04_tran:Vector)
    Rotation=Get pelvis_04_rot() -> (pelvis_04_rot:Rotator)
    Scale="1,1,1"
   ) {
    rotationMode=BMM_Replace
    rotationSpace=BCS_WorldSpace
    scaleSpace=BCS_BoneSpace
    translationMode=BMM_Replace
    translationSpace=BCS_WorldSpace
    alphaBoolBlend {blendOption=Linear}
    alphaScaleBiasClamp {inRange {max=0, min=-40}, outRange {max=0, min=1}}
    boneToModify {boneName=spine_04_ref}
   }
   n15 = "Transform (Modify) Bone - Bone: spine_05_ref"(
    ComponentPose=@n14
    Alpha=1
    Translation=Get pelvis_05_tran() -> (pelvis_05_tran:Vector)
    Rotation=Get pelvis_05_rot() -> (pelvis_05_rot:Rotator)
    Scale="1,1,1"
   ) {
    rotationMode=BMM_Replace
    rotationSpace=BCS_WorldSpace
    scaleSpace=BCS_BoneSpace
    translationMode=BMM_Replace
    translationSpace=BCS_WorldSpace
    alphaBoolBlend {blendOption=Linear}
    alphaScaleBiasClamp {inRange {max=0, min=-40}, outRange {max=0, min=1}}
    boneToModify {boneName=spine_05_ref}
   }
   Save cached pose 'InputCache'(Pose=Component To Local(ComponentPose=@n15)) {cacheName=InputCache}
  ]}
 EventGraph {nodes [
   Event BlueprintUpdateAnimation()
   > n1 = Sequence()
   n2 = GetPartSkeletalMeshComponent()
   Is Valid(exec=@n1.then_1, InputObject=@n2)
   n3 = GetAttachParent!(self:Scene Component Object Reference=@n2) -> (ReturnValue:Scene Component Object Reference)
   > Is Valid(InputObject=@n3)
   > n4 = Sequence!(execute) -> (then_0, then_1, then_2:Exec, then_3:Exec, then_4:Exec, then_5:Exec, then_6:Exec)
   n5 = BreakTransform(InTransform=GetSocketTransform(self=@n3, InSocketName=spine_01, TransformSpace=RTS_World))
   > Set spine_01_tran(spine_01_tran:Vector=@n5) -> (Output_Get:Vector)
   > Set spine_01_rot(spine_01_rot:Rotator=@n5.Rotation) -> (Output_Get:Rotator)
   n6 = BreakTransform(InTransform=GetSocketTransform(self=@n3, InSocketName=spine_02, TransformSpace=RTS_World))
   Set spine_02_tran(execute=@n4.then_1, spine_02_tran:Vector=@n6) -> (Output_Get:Vector)
   > Set spine_02_rot(spine_02_rot:Rotator=@n6.Rotation) -> (Output_Get:Rotator)
   n7 = BreakTransform(InTransform=GetSocketTransform(self=@n3, InSocketName=spine_03, TransformSpace=RTS_World))
   Set spine_03_tran(execute=@n4.then_2, spine_03_tran:Vector=@n7) -> (Output_Get:Vector)
   > Set spine_03_rot(spine_03_rot:Rotator=@n7.Rotation) -> (Output_Get:Rotator)
   n8 = BreakTransform(InTransform=GetSocketTransform(self=@n3, InSocketName=pelvis, TransformSpace=RTS_World))
   Set pelvis_01_tran(execute=[@n4.then_3, @n4.then_4], pelvis_01_tran:Vector=@n8) -> (Output_Get:Vector)
   > Set pelvis_01_rot(pelvis_01_rot:Rotator=@n8.Rotation) -> (Output_Get:Rotator)
   n9 = BreakTransform(InTransform=GetSocketTransform(self=@n3, InSocketName=spine_04, TransformSpace=RTS_World))
   Set pelvis_04_tran(execute=@n4.then_5, pelvis_04_tran:Vector=@n9) -> (Output_Get:Vector)
   > Set pelvis_04_rot(pelvis_04_rot:Rotator=@n9.Rotation) -> (Output_Get:Rotator)
   n10 = BreakTransform(InTransform=GetSocketTransform(self=@n3, InSocketName=spine_05, TransformSpace=RTS_World))
   Set pelvis_05_tran(execute=@n4.then_6, pelvis_05_tran:Vector=@n10) -> (Output_Get:Vector)
   > Set pelvis_05_rot(pelvis_05_rot:Rotator=@n10.Rotation) -> (Output_Get:Rotator)
   n11 = Gravity Override State Function DEPRECATED(
    execute=@n1.then_0
    CharacterPartAnimInstance=Self-Reference()
    GameTimeInSeconds=GetGameTimeInSeconds()
    GravityOverrideParams=Get Gravity Override Params()
   )
   > Set GravityOverride(GravityOverride=@n11)
   n12 = Event OnCharacterCustomizationLoaded()
   > Set Backpack(Backpack:Skeletal Mesh Component Object Reference=@n12.CustomBaseMesh) -> (Output_Get:Skeletal Mesh Component Object Reference)
   > Set Body(Body:Skeletal Mesh Component Budgeted Object Reference=Get Array Item with EFort Custom Part Type(Array=@n12.CustomMeshParts)) -> (Output_Get:Skeletal Mesh Component Budgeted Object Reference)
   n13 = "Get (a copy)"(Array:Array of Custom Character Part Object References=@n12.CustomParts, Dimension 1=1) -> (Output:Custom Character Part Object Reference)
   > Set isF_Med(isF_Med:bool="Equal (Enum)"(
     A:EFortCustomGender Enum=Get GenderPermitted!(self:Custom Character Part Object Reference=@n13) -> (GenderPermitted:EFortCustomGender Enum=Invalid)
     B:EFortCustomGender Enum=Female
    )) -> (Output_Get:bool)
   TryGetPawnOwner()
   "Equal (Enum)"(
    A:EFortCustomBodyType Enum=Get BodyTypesPermitted!(self:Custom Character Part Object Reference=@n13) -> (BodyTypesPermitted:EFortCustomBodyType Enum=NONE)
    B:EFortCustomBodyType Enum=All
   )
  ]}
]
variables [
 Backpack {type=SkeletalMeshComponent}
 Body {type=SkeletalMeshComponentBudgeted}
 Gravity Override Params {type=GravityOverrideParamsStruct}
 GravityOverride {type=Vector}
 isF_Med {type=boolean}
 isFemale {type=boolean}
 pelvis_01_rot {type=Rotator}
 pelvis_01_tran {type=Vector}
 pelvis_04_rot {type=Rotator}
 pelvis_04_tran {type=Vector}
 pelvis_05_rot {type=Rotator}
 pelvis_05_tran {type=Vector}
 spine_01_rot {type=Rotator}
 spine_01_tran {type=Vector}
 spine_02_rot {type=Rotator}
 spine_02_tran {type=Vector}
 spine_03_rot {type=Rotator}
 spine_03_tran {type=Vector}
]
